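# Niagara system for the Silent Tempo emote's default clock effect, layering clock, glow, ring, spinning sector and wave meshes with vortex sparks and a light.
# Meshes follow and orient to User.SkeletalMesh, while User.DynamicFloat indexes arrays that flip orientation, spin and offsets.
# Loop delays and age-driven curves stage the sequence through material parameters, color and alpha.
# Front-end and time-of-day modifiers adjust intensity.
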NiagaraSystem /BRCosmetics/Animation/Game/MainPlayer/Emotes/SilentTempo/FX/NS_Emote_SilentTempo_Clock_Default
dependencies {
 usedDataInterfaces [
  &NiagaraDataInterfaceCurve
  &NiagaraDataInterfaceArrayFloat
  &NiagaraDataInterfaceArrayFloat3
  &NiagaraDataInterfaceColorCurve
 ]
 usedDynamicInputs []
 usedModules [
  &SystemState
  &self:SystemSpawnScript.SystemScriptSource.SystemScriptGraph.NiagaraNodeAssignment_0.SetVariables_#1
 ]
}
emitters {
 Clock {
  emitterUpdateScript [
   &EmitterState(
    Inactive Response=::NewEnumerator1
    Life Cycle Mode=::NewEnumerator1
    Loop Behavior=::NewEnumerator1
    Loop Duration Mode=::NewEnumerator1
    UseLoopDelay=false
   )
   &SpawnBurst_Instantaneous(Use Spawn Probability=false, UseLoopCountLimit=false)
  ]
  particleSpawnScript [
   &PartitionParticles()
   &InitializeParticle(
    Mesh Scale=&Multiply_VectorByFloat(Vector=<1 1 1>, Float=0.85)
    Mesh Scale Mode=::NewEnumerator4
    UsePositionOffset=false
    Write Material Random=true
   )
  ]
  particleUpdateScript [
   &ParticleState(Kill Particles When Lifetime Has Elapsed=false, Let Infinitely Lived Particles Die When Emitter Deactivates=true)
   &SkeletalMeshLocation(Bone Sampling Mode=::NewEnumerator4, MaskSpawnGroup=false, Skeletal Mesh=$User.SkeletalMesh)
   &Timeline(Timeline End=99999997952)
   &DynamicMaterialParameters(
    Index 0 Param 1=&Multiply_Float(A=$Particles.Timeline.Time, B=-1)
    Index 0 Param 2=&Multiply_Float(A=$Particles.Timeline.Time, B=-1.5)
    Index 0 Param 3=&Multiply_Float(A=$Particles.Age, B=0.5)
    Index 0 Param 4=1
    Index 1 Param 1=&FloatFromCurve(
     FloatCurve=DataInterface() {propertyValues {Curve {keys [<at 1.9 0.2>,<at 2 1>]}}}
     CurveIndex=&Subtract_Float(A=$System.Age, B=$System.Delay In)
    )
    Index 1 Param 2=&FloatFromCurve(
     FloatCurve=DataInterface() {propertyValues {Curve {keys [<at 0.8 1>,<at 1.1 0>]}}}
     CurveIndex=&Subtract_Float(A=$System.Age, B=$System.Delay In)
    )
    Index 1 Param 3=&FloatFromCurve(
     FloatCurve=DataInterface() {propertyValues {Curve {keys [
         <at 0 0>
         <at 1 1>
         {arriveTangent=0.7234, interpMode=RCIM_Cubic, leaveTangent=-0.4694, tangentMode=RCTM_Break, time=1.8, value=1}
         {arriveTangent=-0.4679, interpMode=RCIM_Cubic, leaveTangent=-0.4679, tangentMode=RCTM_User, time=3, value=0}
        ]}}}
     CurveIndex=&Subtract_Float(A=$System.Age, B=$System.Delay In)
    )
    Index 1 Param 4=&FloatFromCurve(
     FloatCurve=DataInterface() {propertyValues {Curve {keys [<at 0 0>,<at 1.9 0>,<at 2 1>]}}}
     CurveIndex=&Subtract_Float(A=$System.Age, B=$System.Delay In)
    )
    Index 2 Param 1=&Multiply_Float(
     A=1
     B=&MakeFloatFromVector(Vector=&Multiply_Vector(A=<1 1 1>, B=$Particles.Scale), Channel=::NewEnumerator0)
    )
    Index 2 Param 2=&Multiply_Float(
     A=1
     B=&MakeFloatFromVector(Vector=&Multiply_Vector(A=<1 1 1>, B=$Particles.Scale), Channel=::NewEnumerator0)
    )
    Module.Param0WriteEnabled=true
    Module.Param0WriteEnabled1=true
    Module.Param0WriteEnabled2=true
    Module.Param1WriteEnabled=true
    Module.Param1WriteEnabled1=true
    Module.Param1WriteEnabled2=true
    Module.Param2WriteEnabled=true
    Module.Param2WriteEnabled1=true
    Module.Param2WriteEnabled2=true
    Module.Param3WriteEnabled=false
    Module.Param3WriteEnabled1=true
    Module.Param3WriteEnabled2=true
    Write Parameter Index 1=true
    Write Parameter Index 2=true
   )
   &ScaleColor(
    Scale Alpha=&FN_FrontEndFloatModifier(Gameplay=1, Front End=1.5)
    Scale RGB=&VectorFromFloat(Value=&Multiply_Float(
      A=&FN_FrontEndFloatModifier(
       Gameplay=&FloatFromCurve(
        FloatCurve=DataInterface() {propertyValues {Curve {keys [<at 2 1>,<at 2.1 10>,<at 2.4 1>]}}}
        CurveIndex=&Subtract_Float(A=$System.Age, B=$System.Delay In)
       )
       Front End=&FloatFromCurve(
        FloatCurve=DataInterface() {propertyValues {Curve {keys [<at 2 1>,<at 2.1 4>,<at 2.4 1>]}}}
        CurveIndex=&Subtract_Float(A=$System.Age, B=$System.Delay In)
        Scale Curve=0.75
       )
      )
      B=&Multiply_Float(
       A=&FNTimeOfDayScalar(Night Scale=0.7, Morning Scale=1, Evening Scale=1)
       B=&FN_FrontEndFloatModifier(Gameplay=1, Front End=1)
      )
     ))
    ScaleA=true
    ScaleRGB=true
   )
   /SetVariables(Particles.MeshOrientation=&MultiplyQuaternion(
     Quaternion A=&self:ScratchDynamicInput_03(
      X Vector:Vector3f=<0.01 0.01 1>
      zdir:Vector3f=&RotateVectorByQuaternion(Vector To Rotate=<1 0 0>, Quaternion=$StackContext.SkeletalMeshLocation.SampledRotation)
     )
     Quaternion B=&MakeQuatFromAxisAngle(
      Axis=<1 0 0>
      AngleInDegrees=&SelectFloatFromArray(
       Direct Array Index=&MakeIntFromFloat(Float=$User.DynamicFloat)
       Float Selection Array=DataInterface() {propertyValues {FloatData [0,180]}}
      )
     )
    )) {isSetParameters=true}
  ]
  rendererClasses [&NiagaraMeshRendererProperties]
  renderers [
   {rendererClass=&NiagaraMeshRendererProperties, rendererIndex=0}
   {rendererClass=&NiagaraMeshRendererProperties, rendererIndex=1}
  ]
 }
 Glow {
  emitterUpdateScript [
   &EmitterState(Life Cycle Mode=::NewEnumerator1, Loop Behavior=::NewEnumerator1, Loop Duration Mode=::NewEnumerator1, UseLoopDelay=false)
   &SpawnBurst_Instantaneous(Use Spawn Probability=false, UseLoopCountLimit=false)
  ]
  particleSpawnScript [
   &InitializeParticle(
    Mesh Scale=&Multiply_VectorByFloat(Vector=<2.2 2.2 2.2>, Float=0.75)
    Mesh Scale Mode=::NewEnumerator4
    UsePositionOffset=false
    Write Material Random=true
   )
   &InitialMeshOrientation(Mesh Orientation Mode=::NewEnumerator2, Rotation=<0 -0.25 0>, Use Rotation Vector=true)
  ]
  particleUpdateScript [
   &ParticleState(Kill Particles When Lifetime Has Elapsed=false, Let Infinitely Lived Particles Die When Emitter Deactivates=true)
   &SkeletalMeshLocation(
    Bone Sampling Mode=::NewEnumerator4
    MaskSpawnGroup=false
    Sampled Scale Coordinate Space:ENiagaraCoordinateSpace=::World
    Skeletal Mesh=$User.SkeletalMesh
   )
   &ScaleColor(
    Scale Alpha=&Multiply_Float(
     A=&FloatFromCurve(
      FloatCurve=DataInterface() {propertyValues {Curve {keys [<at 0.6 0>,<at 1.2 1>,<at 2.2 1>,<at 2.7 0>]}}}
      CurveIndex=&Subtract_Float(A=$System.Age, B=$System.Delay In)
     )
     B=&FN_FrontEndFloatModifier(Gameplay=1, Front End=0.25)
    )
    Scale RGB=&VectorFromFloat(Value=&Multiply_Float(
      A=&FloatFromCurve(
       FloatCurve=DataInterface() {propertyValues {Curve {keys [<at 2.4 1>,<at 2.5 1>,<at 2.55 1>,<at 2.65 1>]}}}
       CurveIndex=&Add_Float(A=$System.Age, B=0)
      )
      B=&FNTimeOfDayScalar(Night Scale=0.5, Morning Scale=1, Evening Scale=1)
     ))
    ScaleA=true
    ScaleRGB=true
   )
   /SetVariables(Particles.MeshOrientation=&MultiplyQuaternion(
     Quaternion A=&self:ScratchDynamicInput_03(
      X Vector:Vector3f=<0.01 0.01 1>
      zdir:Vector3f=&RotateVectorByQuaternion(Vector To Rotate=<1 0 0>, Quaternion=$StackContext.SkeletalMeshLocation.SampledRotation)
     )
     Quaternion B=&MakeQuatFromAxisAngle(
      Axis=<1 0 0>
      AngleInDegrees=&SelectFloatFromArray(
       Direct Array Index=&MakeIntFromFloat(Float=$User.DynamicFloat)
       Float Selection Array=DataInterface() {propertyValues {FloatData [0,180]}}
      )
     )
    )) {isSetParameters=true}
  ]
  rendererClasses [&NiagaraMeshRendererProperties]
  renderers [
   {rendererClass=&NiagaraMeshRendererProperties, rendererIndex=0}
   {rendererClass=&NiagaraMeshRendererProperties, rendererIndex=1}
  ]
 }
 Light {
  emitterUpdateScript [
   &EmitterState(Life Cycle Mode=::NewEnumerator1, Loop Behavior=::NewEnumerator1, Loop Delay=1.9, Loop Duration=1, UseLoopDelay=true)
   &SpawnBurst_Instantaneous(Use Spawn Probability=false, UseLoopCountLimit=false)
  ]
  particleSpawnScript [&InitializeParticle(
    Color=<rgba 0 28.3333 100 1>
    Lifetime=2.8
    Sprite Size Mode=::NewEnumerator3
    Uniform Sprite Size=30
    UsePositionOffset=false
    Write Material Random=true
   )]
  particleUpdateScript [
   &ParticleState()
   &ScaleColor(
    Scale Alpha=&FloatFromCurve(FloatCurve=DataInterface() {propertyValues {Curve {keys [<at 0 0>,<at 0.2 1>,<at 0.85 1>,<at 1 0>]}}})
    Scale RGB=&VectorFromFloat(Value=&Multiply_Float(
      A=&FNTimeOfDayScalar(Night Scale=0.5, Morning Scale=1, Evening Scale=1)
      B=&FN_FrontEndFloatModifier(Gameplay=1, Front End=0.25)
     ))
    ScaleA=true
    ScaleRGB=true
   )
   &FN_Light_Attributes(Create Exponent Attribute=true, Create Radius Attribute=true, Exponent=4, Radius=250)
  ]
  rendererClasses [&NiagaraSpriteRendererProperties, &NiagaraLightRendererProperties]
  renderers [
   {rendererClass=&NiagaraSpriteRendererProperties}
   {rendererClass=&NiagaraLightRendererProperties, rendererIndex=1}
  ]
 }
 Ring {
  emitterUpdateScript [
   &EmitterState(
    Inactive Response=::NewEnumerator1
    Life Cycle Mode=::NewEnumerator1
    Loop Behavior=::NewEnumerator1
    Loop Duration Mode=::NewEnumerator1
    UseLoopDelay=false
   )
   &SpawnBurst_Instantaneous(Use Spawn Probability=false, UseLoopCountLimit=false)
  ]
  particleSpawnScript [&InitializeParticle(
    Color=<rgba 1 1 1 0.6>
    Mesh Scale=&Multiply_VectorByFloat(Vector=<1.4733 1.4733 1.4733>, Float=0.75)
    Mesh Scale Mode=::NewEnumerator4
    UsePositionOffset=true
    Write Material Random=true
   )]
  particleUpdateScript [
   &ParticleState(Kill Particles When Lifetime Has Elapsed=false, Let Infinitely Lived Particles Die When Emitter Deactivates=true)
   &SkeletalMeshLocation(Bone Sampling Mode=::NewEnumerator4, MaskSpawnGroup=false, Sampled Position Offset=<10 0 0>, Skeletal Mesh=$User.SkeletalMesh)
   &Timeline(
    Play Rate=&Lerp_Float(A=0.4, B=2, Alpha=$System.InterpolateOverTime.InterpolateOverTime_Float.Moving Average)
    Timeline End=99999997952
   )
   &DynamicMaterialParameters(
    Index 0 Param 1=$Emitter.Age
    Index 0 Param 3=&MakeFloatFromVector(
     Vector=&Multiply_Vector(A=$Particles.SkeletalMeshLocation.SampledScale, B=$Particles.Scale)
     Channel=::NewEnumerator0
    )
    Index 0 Param 4=&MakeFloatFromVector(
     Vector=&Multiply_Vector(A=$Particles.SkeletalMeshLocation.SampledScale, B=$Particles.Scale)
     Channel=::NewEnumerator0
    )
    Index 1 Param 1=1
    Index 1 Param 2=&FloatFromCurve(
     FloatCurve=DataInterface() {propertyValues {Curve {keys [<at 0.7 1>,<at 0.85 0>]}}}
     CurveIndex=&Subtract_Float(A=$System.Age, B=$System.Delay In)
    )
    Index 1 Param 3=&FloatFromCurve(
     FloatCurve=DataInterface() {propertyValues {Curve {keys [<at 0 0>,<at 1 1>]}}}
     CurveIndex=&Subtract_Float(A=$System.Age, B=$System.Delay In)
    )
    Module.Param0WriteEnabled=true
    Module.Param0WriteEnabled1=true
    Module.Param1WriteEnabled=true
    Module.Param1WriteEnabled1=true
    Module.Param2WriteEnabled=true
    Module.Param2WriteEnabled1=true
    Module.Param3WriteEnabled=true
    Module.Param3WriteEnabled1=true
    Write Parameter Index 1=true
   )
   &ScaleColor(
    Scale Alpha=&Multiply_Float(
     A=&FloatFromCurve(
      FloatCurve=DataInterface() {propertyValues {Curve {keys [<at 2 1>,<at 2.5 0>]}}}
      CurveIndex=&Subtract_Float(A=$System.Age, B=$System.Delay In)
     )
     B=1
    )
    Scale RGB=&VectorFromFloat(Value=&Multiply_Float(
      A=&FloatFromCurve(
       FloatCurve=DataInterface() {propertyValues {Curve {keys [<at 2.5 1>]}}}
       CurveIndex=&Subtract_Float(A=$System.Age, B=$System.Delay In)
      )
      B=&FNTimeOfDayScalar(Night Scale=0.7, Morning Scale=1, Evening Scale=1)
     ))
    ScaleA=true
    ScaleRGB=true
   )
   /SetVariables(Particles.MeshOrientation=&MultiplyQuaternion(
     Quaternion A=&self:ScratchDynamicInput_03(
      X Vector:Vector3f=<0.01 0.01 1>
      zdir:Vector3f=&RotateVectorByQuaternion(Vector To Rotate=<1 0 0>, Quaternion=$StackContext.SkeletalMeshLocation.SampledRotation)
     )
     Quaternion B=&MakeQuatFromAxisAngle(
      Axis=<1 0 0>
      AngleInDegrees=&SelectFloatFromArray(
       Direct Array Index=&MakeIntFromFloat(Float=$User.DynamicFloat)
       Float Selection Array=DataInterface() {propertyValues {FloatData [0,180]}}
      )
     )
    )) {isSetParameters=true}
  ]
  rendererClasses [&NiagaraMeshRendererProperties]
  renderers [{rendererClass=&NiagaraMeshRendererProperties, rendererIndex=0}]
 }
 Sector {
  emitterUpdateScript [
   &EmitterState(Inactive Response=::NewEnumerator1, Life Cycle Mode=::NewEnumerator1, Loop Behavior=::NewEnumerator1, Loop Delay=1.5, UseLoopDelay=true)
   &SpawnBurst_Instantaneous(Spawn Time=0.35, Use Spawn Probability=false, UseLoopCountLimit=false)
  ]
  particleSpawnScript [
   &InitializeParticle(
    Lifetime=2
    Mesh Scale=&Multiply_VectorByFloat(Vector=<1.15 1.15 0.3>, Float=0.75)
    Mesh Scale Mode=::NewEnumerator4
    UsePositionOffset=false
    Write Material Random=true
   )
   &CameraOffset(Camera Offset Amount=&FN_FrontEndFloatModifier(Gameplay=10, Front End=2))
  ]
  particleUpdateScript [
   &ParticleState()
   &SkeletalMeshLocation(Bone Sampling Mode=::NewEnumerator4, MaskSpawnGroup=false, Skeletal Mesh=$User.SkeletalMesh)
   &ScaleColor(
    Scale Alpha=&FloatFromCurve(
     FloatCurve=DataInterface() {propertyValues {Curve {keys [
         {interpMode=RCIM_Cubic, tangentMode=RCTM_User, time=0, value=0}
         {interpMode=RCIM_Cubic, tangentMode=RCTM_User, time=0.25, value=1}
         {interpMode=RCIM_Cubic, tangentMode=RCTM_User, time=1, value=0}
        ]}}}
     Scale Curve=&FN_FrontEndFloatModifier(Gameplay=1, Front End=0.75)
    )
    Scale RGB=&VectorFromFloat(Value=&FNTimeOfDayScalar(Night Scale=0.75, Morning Scale=1, Evening Scale=1))
    ScaleA=true
    ScaleRGB=true
   )
   /SetVariables(Particles.MeshOrientation=&MultiplyQuaternion(
     Quaternion A=&self:ScratchDynamicInput_03(
      X Vector:Vector3f=<0.01 0.01 1>
      zdir:Vector3f=&RotateVectorByQuaternion(Vector To Rotate=<1 0 0>, Quaternion=$StackContext.SkeletalMeshLocation.SampledRotation)
     )
     Quaternion B=&MultiplyQuaternion(
      Quaternion A=&MakeQuatFromAxisAngle(
       Axis=&SelectVectorFromArray(
        Direct Array Index=&MakeIntFromFloat(Float=$User.DynamicFloat)
        Vector Selection Array=DataInterface() {propertyValues {FloatData [<0 0 1>, <0 0 -1>]}}
       )
       AngleInDegrees=&Add_Float(A=&Multiply_Float(A=$Emitter.Age, B=&Multiply_Float(A=360, B=1.53)), B=95)
      )
      Quaternion B=&MakeQuatFromAxisAngle(
       Axis=<1 0 0>
       AngleInDegrees=&SelectFloatFromArray(
        Direct Array Index=&MakeIntFromFloat(Float=$User.DynamicFloat)
        Float Selection Array=DataInterface() {propertyValues {FloatData [0,180]}}
       )
      )
     )
    )) {isSetParameters=true}
  ]
  rendererClasses [&NiagaraMeshRendererProperties]
  renderers [{rendererClass=&NiagaraMeshRendererProperties, rendererIndex=0}]
 }
 Spark_Ring {
  emitterUpdateScript [
   &EmitterState(Life Cycle Mode=::NewEnumerator1, Loop Behavior=::NewEnumerator1, Loop Delay=2.05, Loop Duration=2.1, UseLoopDelay=true)
   &SpawnRate(
    SpawnRate=&FloatFromCurve(
     FloatCurve=DataInterface() {propertyValues {Curve {keys [
         {interpMode=RCIM_Cubic, tangentMode=RCTM_User, time=0.5, value=1}
         {arriveTangent=-2.7246, interpMode=RCIM_Cubic, leaveTangent=-2.7246, tangentMode=RCTM_User, time=1, value=0}
        ]}}}
     CurveIndex=$Emitter.NormalizedLoopAge
     Scale Curve=400
    )
    Use Spawn Probability=false
   )
   &SpawnBurst_Instantaneous(Spawn Count=20, Use Spawn Probability=false, UseLoopCountLimit=false)
  ]
  particleSpawnScript [
   &InitializeParticle(
    Color=<rgba 0.5 3.6667 10 1>
    Lifetime Max=1.2
    Lifetime Min=0.5
    Lifetime Mode=::NewEnumerator1
    Sprite Size Max=<4 22>
    Sprite Size Min=<3 15>
    Sprite Size Mode=::NewEnumerator2
    UsePositionOffset=false
    Write Material Random=true
   )
   &SkeletalMeshLocation(Bone Sampling Mode=::NewEnumerator4, Interpolate Mesh Sampling=false, MaskSpawnGroup=false, Skeletal Mesh=$User.SkeletalMesh)
   &ShapeLocation(
    MaskSpawnGroup=false
    Override Seed=false
    Ring Radius=&RandomRangeFloat(Minimum=80, Maximum=80)
    Rotation Mode=::NewEnumerator1
    Shape Primitive=::NewEnumerator4
    UseOwnerScale=true
   )
   &AddVelocity(
    Module.Use Falloff Distance=false
    Override Seed=false
    Velocity Mode=::NewEnumerator1
    Velocity Speed=&Multiply_Float(A=&RandomRangeFloat(Minimum=2.5, Maximum=3.5), B=100)
   )
  ]
  particleUpdateScript [
   &ParticleState()
   /Niagara/Modules/Update/Velocity/VortexVelocity(
    Influence Falloff Exponent=0
    Influence Falloff Radius=0
    Module.Use Influence Falloff=true
    Velocity Amount:NiagaraFloat=&Multiply_Float(
     A=&Multiply_Float(
      A=&RandomRangeFloat(Minimum=6, Maximum=2.5)
      B=&SelectFloatFromArray(
       Direct Array Index=&MakeIntFromFloat(Float=$User.DynamicFloat)
       Float Selection Array=DataInterface() {propertyValues {FloatData [1,-1]}}
      )
     )
     B=&Multiply_Float(
      A=20000
      B=&FloatFromCurve(FloatCurve=DataInterface() {propertyValues {Curve {keys [<at 0 0>,<at 0.25 1>,<at 1 0>]}}})
     )
    )
    Vortex Axis:Vector3f=<1 0 0>
   )
   &Drag(
    Drag=&FN_FrontEndFloatModifier(Gameplay=150, Front End=100)
    Use Linear Drag=true
    Use Rotational Drag=true
   )
   &CurlNoiseForce(
    Noise Strength=&FloatFromCurve(FloatCurve=DataInterface() {propertyValues {Curve {keys [<at 0 0>,<at 1 1>]}}}, Scale Curve=100000)
    Pan Noise=false
   )
   &SolveForcesAndVelocity(Clamp Velocity=false, Limit Acceleration=false)
   &ScaleSpriteSizeBySpeed(Max Scale Factor=<1 1.5>, Min Scale Factor=<0.5 0>)
   &ScaleSpriteSize(
    Uniform Curve Scale=0.6
    Uniform Curve Sprite Scale=DataInterface() {propertyValues {Curve {keys [{interpMode=RCIM_Cubic, tangentMode=RCTM_User, time=0.4, value=1}, {interpMode=RCIM_Cubic, tangentMode=RCTM_User, time=1, value=0}]}}}
   )
   /SetVariables(
    Particles.Color2=&MultiplyLinearColor(
     Linear Color=$Particles.Color
     Scale Factor=&MakeFloat4FromVectorAndFloat("Vector (RGB)"=<1 1 1>, "Float (Alpha)"=0.05)
    )
    Particles.SpriteSize2=&Multiply_Vector2D(A=$Particles.SpriteSize, B=<8 2>)
   ) {isSetParameters=true}
  ]
  rendererClasses [&NiagaraSpriteRendererProperties]
  renderers [
   {rendererClass=&NiagaraSpriteRendererProperties}
   {rendererClass=&NiagaraSpriteRendererProperties, rendererIndex=1}
  ]
 }
 Wave {
  emitterUpdateScript [
   &EmitterState(Life Cycle Mode=::NewEnumerator1, Loop Behavior=::NewEnumerator1, Loop Delay=2.9, Loop Duration=1, UseLoopDelay=true)
   &SpawnRate(SpawnRate=2, Use Spawn Probability=false)
  ]
  particleSpawnScript [/Niagara/Modules/Spawn/Initialization/InitializeParticle(
    Color=<rgba 1.5 0.3114 1.0299 2>
    Lifetime=0.8
    Mesh Scale=&Multiply_Vector(
     A=&SelectVectorFromArray(
      Direct Array Index=&MakeIntFromFloat(Float=$User.DynamicFloat)
      Vector Selection Array=DataInterface() {propertyValues {FloatData [<4 3 3>, <-4 3 3>]}}
     )
     B=<1 0.75 0.75>
    )
    Module.Write Color=true
    Module.Write Lifetime=true
    Module.Write Mass=false
    Module.Write Material Random=true
    Module.Write Position=false
    Module.Write Scale=true
    Module.Write SpriteRotation=false
    Module.Write SpriteSize=false
    Position=&SimulationPosition()
    Sprite Rotation=&UniformRangedFloat(Minimum=0, Maximum=1)
   )]
  particleUpdateScript [
   &ParticleState()
   &ScaleMeshSize(Scale Factor=&Multiply_VectorByFloat(
     Vector=&Multiply_VectorByFloat(Vector=<1 1 1>, Float=&FN_FrontEndFloatModifier(Gameplay=1, Front End=0.9))
     Float=&Lerp_Float(A=0.8, B=1, Alpha=&ClampFloat(Float=$User.ParaGlideLeanAlpha))
    )) {module=ScaleMeshSize001}
   &SolveForcesAndVelocity(Clamp Velocity=false, Limit Acceleration=false)
   &ScaleColor(
    Scale Alpha=&Multiply_Float(
     A=&FloatFromCurve(FloatCurve=DataInterface() {propertyValues {Curve {keys [
          {arriveTangent=4.0516, interpMode=RCIM_Cubic, leaveTangent=4.0516, tangentMode=RCTM_User, time=0, value=0}
          {arriveTangent=0.0961, interpMode=RCIM_Cubic, leaveTangent=0.0961, tangentMode=RCTM_User, time=0.25, value=0.5}
          <at 0.5 0.5>
          {arriveTangent=0.2272, interpMode=RCIM_Cubic, leaveTangent=0.2272, tangentMode=RCTM_User, time=1, value=0}
         ]}}})
     B=&FNTimeOfDayScalar(Night Scale=0.5, Morning Scale=1, Evening Scale=1)
    )
    Scale RGB=&MakeVectorFromLinearColorRGB(Color=&ColorFromCurve(ColorCurve=DataInterface() {propertyValues {
        AlphaCurve {keys [<at 0 1 RCIM_Cubic>]}
        BlueCurve {keys [<at 0 1 RCIM_Cubic>,<at 0.222 0.7891 RCIM_Cubic>]}
        GreenCurve {keys [<at 0 0.6104 RCIM_Cubic>,<at 0.222 1 RCIM_Cubic>]}
        RedCurve {keys [<at 0 0.0396 RCIM_Cubic>,<at 0.222 0.088 RCIM_Cubic>]}
       }}))
    ScaleA=true
    ScaleRGB=true
   )
   &SkeletalMeshLocation(
    Bone Sampling Mode=::NewEnumerator4
    MaskSpawnGroup=false
    Sampled Position Offset=&SelectVectorFromArray(
     Direct Array Index=&MakeIntFromFloat(Float=$User.DynamicFloat)
     Vector Selection Array=DataInterface() {propertyValues {FloatData [<-70 0 0>, <70 0 0>]}}
    )
    Sampled Scale Coordinate Space:ENiagaraCoordinateSpace=::World
    Skeletal Mesh=$User.SkeletalMesh
   )
   /SetVariables(Particles.MeshOrientation=&MultiplyQuaternion(
     Quaternion A=&self:ScratchDynamicInput_03(
      X Vector:Vector3f=<0.01 0.01 1>
      zdir:Vector3f=&RotateVectorByQuaternion(Vector To Rotate=<1 0 0>, Quaternion=$StackContext.SkeletalMeshLocation.SampledRotation)
     )
     Quaternion B=&MakeQuatFromAxisAngle(Axis=<0 1 0>, AngleInDegrees=90)
    )) {isSetParameters=true}
  ]
  rendererClasses [&NiagaraMeshRendererProperties]
  renderers [{rendererClass=&NiagaraMeshRendererProperties, rendererIndex=0}]
 }
}
summary {
 systemName=NS_Emote_SilentTempo_Clock_Default
 emitters [
  {emitterName=Ring, rendererClasses [&NiagaraMeshRendererProperties]}
  {emitterName=Glow, rendererClasses [&NiagaraMeshRendererProperties]}
  {emitterName=Wave, rendererClasses [&NiagaraMeshRendererProperties]}
  {emitterName=Sector, rendererClasses [&NiagaraMeshRendererProperties]}
  {emitterName=Spark_Ring, rendererClasses [&NiagaraSpriteRendererProperties]}
  {emitterName=Light, rendererClasses [&NiagaraSpriteRendererProperties, &NiagaraLightRendererProperties]}
  {emitterName=Clock, rendererClasses [&NiagaraMeshRendererProperties]}
 ]
 userVariables [
  User.bIsFrontEnd {type=NiagaraBool}
  User.bIsFrontEndPreview {type=NiagaraBool}
  User.DynamicFloat {type=NiagaraFloat, defaultValue=0, description=""}
  User.ParaGlideLeanAlpha {type=NiagaraFloat, defaultValue=0, description=""}
  User.SkeletalMesh {
   type=NiagaraDataInterfaceSkeletalMesh
   defaultValue {
    struct=NiagaraExt_VariableValue_DataInterface
    value {dataInterface=&self:NiagaraDataInterfaceSkeletalMesh_0}
   }
  }
 ]
}
