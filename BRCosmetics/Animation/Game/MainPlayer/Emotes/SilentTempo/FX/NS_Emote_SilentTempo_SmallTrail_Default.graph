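# Niagara effect for the SilentTempo emote that adds small, short-lived blue glowing trails to points on the player's skeletal mesh.
# Partitioned Flare, ParticleDecal and Trail02 emitters put five sprites or ribbons on bones or sockets of User.SkeletalMesh, fading them via curves.
# ZeroPointFlare is a delayed three-layer sprite flash shifted by User.Offset, its spawn probability falling as User.DynamicFloat rises.
# Brightness scales by front-end context and, except for ZeroPointFlare, time of day.
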NiagaraSystem /BRCosmetics/Animation/Game/MainPlayer/Emotes/SilentTempo/FX/NS_Emote_SilentTempo_SmallTrail_Default
dependencies {
 usedDataInterfaces [&NiagaraDataInterfaceCurve, &NiagaraDataInterfaceColorCurve]
 usedDynamicInputs []
 usedModules [&SystemState]
}
emitters {
 Flare {
  emitterUpdateScript [
   &EmitterState(Inactive Response=::NewEnumerator1, Life Cycle Mode=::NewEnumerator1, Loop Behavior=::NewEnumerator1, Loop Duration=0.7, UseLoopDelay=false)
   &SpawnBurst_Instantaneous(Spawn Count=5, Use Spawn Probability=false, UseLoopCountLimit=false)
  ]
  particleSpawnScript [
   &PartitionParticles(Number Of Partitions=5)
   &InitializeParticle(
    Color=<rgba 0.3 0.4167 1 0.8>
    Lifetime=3
    Sprite Size Mode=::NewEnumerator3
    Uniform Sprite Size=50
    UsePositionOffset=false
    Write Material Random=true
   )
   &CameraOffset(Camera Offset Amount=10)
  ]
  particleUpdateScript [
   &ParticleState()
   &SkeletalMeshLocation(
    Bone / Socket Index=$Particles.PartitionParticles.Partition
    Bone Sampling Mode=::NewEnumerator4
    MaskSpawnGroup=false
    Skeletal Mesh=$User.SkeletalMesh
   )
   &ScaleColor(
    Scale Alpha=&FloatFromCurve(FloatCurve=DataInterface() {propertyValues {Curve {keys [
         {interpMode=RCIM_Cubic, tangentMode=RCTM_User, time=0, value=0}
         {interpMode=RCIM_Cubic, tangentMode=RCTM_User, time=0.2, value=1}
         {interpMode=RCIM_Cubic, tangentMode=RCTM_User, time=0.7, value=0}
        ]}}})
    Scale RGB=&VectorFromFloat(Value=&Multiply_Float(
      A=&FNTimeOfDayScalar(Night Scale=0.7, Morning Scale=1, Evening Scale=1)
      B=&FN_FrontEndFloatModifier(Gameplay=1, Front End=0.5)
     ))
    ScaleA=true
    ScaleRGB=true
   )
  ]
  rendererClasses [&NiagaraSpriteRendererProperties]
  renderers [{rendererClass=&NiagaraSpriteRendererProperties}]
 }
 ParticleDecal {
  emitterUpdateScript [
   &EmitterState(Life Cycle Mode=::NewEnumerator1, Loop Behavior=::NewEnumerator1, Loop Duration=0.8, UseLoopDelay=false)
   &SpawnBurst_Instantaneous(Spawn Count=5, Use Spawn Probability=false, UseLoopCountLimit=false)
  ]
  particleSpawnScript [
   &PartitionParticles(Number Of Partitions=5)
   &InitializeParticle(
    Color=<rgba 3 22 60 1>
    Lifetime=3
    Sprite Size Mode=::NewEnumerator3
    Uniform Sprite Size=30
    UsePositionOffset=false
    Write Material Random=true
   )
   &CameraOffset(Camera Offset Amount=5)
  ]
  particleUpdateScript [
   &ParticleState()
   &DynamicMaterialParameters(
    Index 0 Param 1=1
    Index 0 Param 4=&RandomRangeFloat(Minimum=0, Maximum=1)
    Module.Param0WriteEnabled=true
    Module.Param0WriteEnabled1=true
    Module.Param1WriteEnabled=true
    Module.Param1WriteEnabled1=true
    Module.Param2WriteEnabled=true
    Module.Param2WriteEnabled1=true
    Module.Param3WriteEnabled=true
    Module.Param3WriteEnabled1=true
    Write Parameter Index 1=true
   )
   &SkeletalMeshLocation(
    Bone / Socket Index=$Particles.PartitionParticles.Partition
    Bone Sampling Mode=::NewEnumerator4
    MaskSpawnGroup=false
    Skeletal Mesh=$User.SkeletalMesh
   )
   &ScaleColor(
    Scale Alpha=&FloatFromCurve(FloatCurve=DataInterface() {propertyValues {Curve {keys [
         {interpMode=RCIM_Cubic, tangentMode=RCTM_User, time=0, value=0}
         {interpMode=RCIM_Cubic, tangentMode=RCTM_User, time=0.2, value=1}
         {interpMode=RCIM_Cubic, tangentMode=RCTM_User, time=1, value=0}
        ]}}})
    Scale RGB=&VectorFromFloat(Value=&Multiply_Float(
      A=&FNTimeOfDayScalar(Night Scale=0.7, Morning Scale=1, Evening Scale=1)
      B=&FN_FrontEndFloatModifier(Gameplay=1, Front End=1)
     ))
    ScaleA=true
    ScaleRGB=true
   )
  ]
  rendererClasses [&NiagaraSpriteRendererProperties]
  renderers [{rendererClass=&NiagaraSpriteRendererProperties}]
 }
 Trail02 {
  emitterUpdateScript [
   &EmitterState(Life Cycle Mode=::NewEnumerator1, Loop Behavior=::NewEnumerator1, Loop Duration=0.75, UseLoopDelay=false)
   &FN_SpawnRibbon_v2(Number of Ribbons=5, Spawn Mode=::NewEnumerator0, SpawnRate=100)
  ]
  particleSpawnScript [
   &PartitionParticles(Number Of Partitions=5)
   &FN_InitializeRibbon_v3(Enable Skeletal Position=true, Ribbon Width=40)
   &SkeletalMeshLocation(
    Bone / Socket Index=$Particles.PartitionParticles.Partition
    Bone Sampling Mode=::NewEnumerator4
    MaskSpawnGroup=false
    Skeletal Mesh=$User.SkeletalMesh
   )
  ]
  particleUpdateScript [
   &ParticleState()
   &ScaleRibbonWidth(Ribbon Width Scale=&FloatFromCurve(FloatCurve=DataInterface() {propertyValues {Curve {keys [<at 0 1>,<at 1 0>]}}}))
   &Color(Color=&ColorFromCurve(ColorCurve=DataInterface() {propertyValues {
       AlphaCurve {keys [<at 0 1 RCIM_Cubic>]}
       BlueCurve {keys [<at 0.022 5>,<at 0.5804 2>]}
       GreenCurve {keys [<at 0.022 5>,<at 0.5804 1>]}
       RedCurve {keys [<at 0.022 1>,<at 0.5804 0>]}
      }}))
   &ScaleColor(
    Color Value To Scale=$Particles.Color
    Scale Alpha=&FloatFromCurve(
     FloatCurve=DataInterface() {propertyValues {Curve {keys [<at 0.5 1>,<at 0.7 0>]}}}
     CurveIndex=&Add_Float(A=$Emitter.Age, B=0)
    )
    Scale RGB=&VectorFromFloat(Value=&Multiply_Float(
      A=&FNTimeOfDayScalar(Night Scale=0.7, Morning Scale=1, Evening Scale=1)
      B=&FN_FrontEndFloatModifier(Gameplay=1, Front End=1)
     ))
    ScaleA=true
    ScaleRGB=true
   )
   &DynamicMaterialParameters(
    Index 0 Param 1=1
    Index 0 Param 2=1
    Index 0 Param 3=1
    Module.Param0WriteEnabled=true
    Module.Param1WriteEnabled=true
    Module.Param2WriteEnabled=true
    Module.Param3WriteEnabled=true
   )
   &SolveForcesAndVelocity(Clamp Velocity=false, Limit Acceleration=false)
   &FN_UpdateRibbon(Detach Locked Particle=true, Enable Skeletal Position=true)
  ]
  rendererClasses [&NiagaraRibbonRendererProperties]
  renderers [{rendererClass=&NiagaraRibbonRendererProperties}]
 }
 ZeroPointFlare {
  emitterUpdateScript [
   &EmitterState(Life Cycle Mode=::NewEnumerator1, Loop Behavior=::NewEnumerator1, Loop Delay=0.64, Loop Duration=1, UseLoopDelay=true)
   &SpawnBurst_Instantaneous(
    Spawn Probability=&Lerp_Float(A=1, B=0, Alpha=$User.DynamicFloat)
    Use Spawn Probability=true
    UseLoopCountLimit=false
   )
  ]
  particleSpawnScript [
   /Niagara/Modules/Spawn/Initialization/InitializeParticle(
    Color=<rgba 0 0.6667 1 1>
    Lifetime=0.25
    Module.Write Color=true
    Module.Write Lifetime=true
    Module.Write Mass=true
    Module.Write Material Random=true
    Module.Write Position=false
    Module.Write Scale=false
    Module.Write SpriteRotation=true
    Module.Write SpriteSize=true
    Sprite Rotation=&RandomRangeFloat(Minimum=0, Maximum=360)
    Sprite Size=&Vector2DFromFloat(Value=100)
   )
   /SetVariables(Particles.Position=&AddVectorToPosition(Position=$Particles.Position, Vector=$User.Offset)) {isSetParameters=true}
   &CameraOffset()
  ]
  particleUpdateScript [
   &ParticleState()
   &ScaleColor(
    Scale Alpha=&FloatFromCurve(FloatCurve=DataInterface() {propertyValues {Curve {keys [<at 0 1 RCIM_Cubic>,<at 0.25 1 RCIM_Cubic>,<at 0.75 1 RCIM_Cubic>,<at 1 1>]}}})
    Scale RGB=&VectorFromFloat(Value=&FN_FrontEndFloatModifier(Gameplay=1, Front End=0.75))
    ScaleA=true
    ScaleRGB=true
   )
   /SetVariables(Particles.Size:NiagaraFloat=&FloatFromCurve(FloatCurve=DataInterface() {propertyValues {Curve {keys [<at -0.0039 0.0032 RCIM_Cubic>,<at 1 0.8>]}}})) {isSetParameters=true}
   &FN_Sprite_Attributes(
    Camera Offset=30
    Color=&MakeLinearColorFromVectorAndFloat(
     "Vector (RGB)"=<0.8602 1.4152 1>
     "Float (Alpha)"=&FloatFromCurve(
      FloatCurve=DataInterface() {propertyValues {Curve {keys [{arriveTangent=-1.7159, interpMode=RCIM_Cubic, leaveTangent=-1.7159, tangentMode=RCTM_User, time=0, value=0.9786}, <at 0.9988 -0.0024>]}}}
      Scale Curve=&FloatFromCurve(
       FloatCurve=DataInterface() {propertyValues {Curve {keys [{arriveTangent=-2.325, interpMode=RCIM_Cubic, leaveTangent=-2.325, tangentMode=RCTM_User, time=0, value=0.7921}, <at 0.5 -0.0428>]}}}
       CurveIndex=$Emitter.NormalizedLoopAge
       Scale Curve=0.2
      )
     )
    )
    Create Camera Offset Attribute=true
    Create Color Attribute=true
    Create SpriteSize Attribute=true
    SpriteSize=&Vector2DFromFloat(Value=&Multiply_Float(
      A=70
      B=&FloatFromCurve(FloatCurve=DataInterface() {propertyValues {Curve {keys [<at 0 1.5>, {arriveTangent=1.6365, leaveTangent=1.6365, time=1, value=2}]}}})
     ))
    Write to Module Namespace=true
   ) {module=FN_Sprite_Attributes002}
   &FN_Sprite_Attributes(
    Camera Offset=10
    Color=&ColorFromCurve(ColorCurve=DataInterface() {propertyValues {
       AlphaCurve {keys [<at 0 1 RCIM_Cubic>,<at 0.7586 0 RCIM_Cubic>]}
       BlueCurve {keys [<at 0 1 RCIM_Cubic>]}
       GreenCurve {keys [<at 0 0.1667 RCIM_Cubic>]}
       RedCurve {keys [<at 0 0 RCIM_Cubic>]}
      }})
    Create Camera Offset Attribute=true
    Create Color Attribute=true
    Create Sprite Rotation Attribute=true
    Create SpriteSize Attribute=true
    SpriteSize=&Multiply_Vector2(A=<200 200>, B=<1 1>)
    Write to Module Namespace=true
   ) {module=FN_Sprite_Attributes003}
   &FN_Sprite_Attributes(
    Camera Offset=20
    Color=<rgba 0.4 1.4667 2 1>
    Create Camera Offset Attribute=true
    Create Color Attribute=true
    Create Sprite Rotation Attribute=true
    Create SpriteSize Attribute=true
    SpriteSize=&Multiply_Vector2(
     A=<60 20>
     B=&Vector2DFromFloat(Value=&FloatFromCurve(
       FloatCurve=DataInterface() {propertyValues {Curve {keys [
           <at 0 0 RCIM_Cubic>
           {arriveTangent=-4.3676, interpMode=RCIM_Cubic, leaveTangent=-4.3676, tangentMode=RCTM_User, time=0.3, value=1.15}
           <at 1 0>
          ]}}}
       Scale Curve=5
      ))
    )
    Write to Module Namespace=true
   ) {module=FN_Sprite_Attributes004}
  ]
  rendererClasses [&NiagaraSpriteRendererProperties]
  renderers [
   {rendererClass=&NiagaraSpriteRendererProperties}
   {rendererClass=&NiagaraSpriteRendererProperties, rendererIndex=1}
   {rendererClass=&NiagaraSpriteRendererProperties, rendererIndex=2}
  ]
 }
}
summary {
 systemName=NS_Emote_SilentTempo_SmallTrail_Default
 emitters [
  {emitterName=Trail02, rendererClasses [&NiagaraRibbonRendererProperties]}
  {emitterName=ParticleDecal, rendererClasses [&NiagaraSpriteRendererProperties]}
  {emitterName=ZeroPointFlare, rendererClasses [&NiagaraSpriteRendererProperties]}
  {emitterName=Flare, rendererClasses [&NiagaraSpriteRendererProperties]}
 ]
 userVariables [
  User.bIsFrontEnd {type=NiagaraBool}
  User.bIsFrontEndPreview {type=NiagaraBool}
  User.DynamicFloat {type=NiagaraFloat, defaultValue=0, description=""}
  User.GlobalScale {type=NiagaraFloat, defaultValue=1, description=""}
  User.Offset {type=Vector3f, description="", defaultValue {struct=Vector3f, value=<0 0 0>}}
  User.RibbonIDOffset {type=NiagaraInt32}
  User.SkeletalMesh {
   type=NiagaraDataInterfaceSkeletalMesh
   defaultValue {
    struct=NiagaraExt_VariableValue_DataInterface
    value {dataInterface=&self:NiagaraDataInterfaceSkeletalMesh_2}
   }
  }
 ]
}
